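# Sound cue for the cook-in of the TidalCraneDropPod item, classed as a gadget item sound.
# A mixer layers five random picks: throw whoosh, run foley, pod cook, flare in-air loop and flare activate/land, with the cook layer loudest and the loop lowest.
# The in-air waves loop, so the cue keeps playing until stopped.
# Spatialization borrows the default grenade cook attenuation, with reduced overall volume.
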SoundCue /TidalCraneDropPod/Assets/Audio/SC_TidalCraneDropPod_CookIn
root=Mixer_1
nodes [
 {id=Mixer_1, children [Random_1,Random_3,Random_0,Random_2,Random_4], properties {inputVolume [0.5,0.5,1,0.35,0.5]}}
 {id=Random_0, children [WavePlayer_0,WavePlayer_2,WavePlayer_3], properties {weights [1,1,1]}}
 {id=Random_1, children [WavePlayer_9, WavePlayer_10, WavePlayer_11, WavePlayer_12, WavePlayer_13], properties {weights [1,1,1,1,1]}}
 {id=Random_2, children [WavePlayer_4,WavePlayer_5], properties {weights [1,1]}}
 {id=Random_3, children [WavePlayer_1, WavePlayer_15, WavePlayer_16, WavePlayer_17], properties {weights [1,1,1,1]}}
 {id=Random_4, children [WavePlayer_6,WavePlayer_7,WavePlayer_8], properties {weights [1,1,1]}}
 {id=WavePlayer_0, properties {soundWaveAssetPtr=/TidalCraneDropPod/Assets/Audio/wav/TidalCraneDropPod_Cook_01}}
 {id=WavePlayer_1, properties {soundWaveAssetPtr=/TidalCraneDropPod/Assets/Audio/wav/player_run_foley_15}}
 {id=WavePlayer_10, properties {soundWaveAssetPtr=/TidalCraneDropPod/Assets/Audio/wav/Foley_Minifig_Throw_2H_Whoosh_05}}
 {id=WavePlayer_11, properties {soundWaveAssetPtr=/TidalCraneDropPod/Assets/Audio/wav/Foley_Minifig_Throw_2H_Whoosh_04}}
 {id=WavePlayer_12, properties {soundWaveAssetPtr=/TidalCraneDropPod/Assets/Audio/wav/Foley_Minifig_Throw_2H_Whoosh_03}}
 {id=WavePlayer_13, properties {soundWaveAssetPtr=/TidalCraneDropPod/Assets/Audio/wav/Foley_Minifig_Throw_2H_Whoosh_02}}
 {id=WavePlayer_15, properties {soundWaveAssetPtr=/TidalCraneDropPod/Assets/Audio/wav/player_run_foley_14}}
 {id=WavePlayer_16, properties {soundWaveAssetPtr=/TidalCraneDropPod/Assets/Audio/wav/player_run_foley_13}}
 {id=WavePlayer_17, properties {soundWaveAssetPtr=/TidalCraneDropPod/Assets/Audio/wav/player_run_foley_12}}
 {id=WavePlayer_2, properties {soundWaveAssetPtr=/TidalCraneDropPod/Assets/Audio/wav/TidalCraneDropPod_Cook_02}}
 {id=WavePlayer_3, properties {soundWaveAssetPtr=/TidalCraneDropPod/Assets/Audio/wav/TidalCraneDropPod_Cook_03}}
 {id=WavePlayer_4, properties {bLooping=true, soundWaveAssetPtr=/TidalCraneDropPod/Assets/Audio/wav/FlareGun_InAirLoop_01}}
 {id=WavePlayer_5, properties {bLooping=true, soundWaveAssetPtr=/TidalCraneDropPod/Assets/Audio/wav/FlareGun_InAirLoop_02}}
 {id=WavePlayer_6, properties {soundWaveAssetPtr=/TidalCraneDropPod/Assets/Audio/wav/TidalCraneDropPod_FlareActivateLand_01}}
 {id=WavePlayer_7, properties {soundWaveAssetPtr=/TidalCraneDropPod/Assets/Audio/wav/TidalCraneDropPod_FlareActivateLand_02}}
 {id=WavePlayer_8, properties {soundWaveAssetPtr=/TidalCraneDropPod/Assets/Audio/wav/TidalCraneDropPod_FlareActivateLand_03}}
 {id=WavePlayer_9, properties {soundWaveAssetPtr=/TidalCraneDropPod/Assets/Audio/wav/Foley_Minifig_Throw_2H_Whoosh_06}}
]
properties {
 attenuationSettings=/Game/Sounds/Fort_Weapons/Grenades/Grenade_Default_Cook_Att
 maxDistance=3000
 soundClassObject=&Fort_Gadgets_Items
 volumeMultiplier=0.4
}
