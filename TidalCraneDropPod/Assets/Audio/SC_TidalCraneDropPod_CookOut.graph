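# Sound cue for the Tidal Crane drop pod's cook-out sound, playing one of four player-run foley wave variants.
# A random node with equal weights picks among the four wave players on each play.
# It plays quietly through a reduced volume multiplier in the gadget items sound class, spatialized with the default grenade cook attenuation.
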SoundCue /TidalCraneDropPod/Assets/Audio/SC_TidalCraneDropPod_CookOut
root=Random_3
nodes [
 {id=Random_3, children [WavePlayer_1, WavePlayer_15, WavePlayer_16, WavePlayer_17], properties {weights [1,1,1,1]}}
 {id=WavePlayer_1, properties {soundWaveAssetPtr=/TidalCraneDropPod/Assets/Audio/wav/player_run_foley_15}}
 {id=WavePlayer_15, properties {soundWaveAssetPtr=/TidalCraneDropPod/Assets/Audio/wav/player_run_foley_14}}
 {id=WavePlayer_16, properties {soundWaveAssetPtr=/TidalCraneDropPod/Assets/Audio/wav/player_run_foley_13}}
 {id=WavePlayer_17, properties {soundWaveAssetPtr=/TidalCraneDropPod/Assets/Audio/wav/player_run_foley_12}}
]
properties {
 attenuationSettings=/Game/Sounds/Fort_Weapons/Grenades/Grenade_Default_Cook_Att
 maxDistance=3000
 soundClassObject=&Fort_Gadgets_Items
 volumeMultiplier=0.15
}
