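# Sound cue for the Tidal Crane Drop Pod's drop sound.
# A single wave player plays the drop wave directly, with no randomization, mixing or modulation nodes.
# It uses the Fortnite weapon-drop attenuation and the gadget items sound class, with reduced volume and a limited max audible distance.
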SoundCue /TidalCraneDropPod/Assets/Audio/SC_TidalCraneDropPod_Drop
root=WavePlayer_1
nodes [{id=WavePlayer_1, properties {soundWaveAssetPtr=/TidalCraneDropPod/Assets/Audio/wav/TidalCraneDropPod_Drop}}]
properties {
 attenuationSettings=&Fort_Weapon_Drop_Att
 maxDistance=3000
 soundClassObject=&Fort_Gadgets_Items
 volumeMultiplier=0.7
}
