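# Sound Cue for the Tidal Crane drop pod's rift-entry sound, layering rift impacts with an eagle whistle.
# A mixer blends a fixed EnterRiftA wave, random B and pitch-modulated C variants, and a quieter random pick from four whistles.
# Rift waves and high-frequency alert attenuation come from DropPodParentGameplay, while the whistles are local to this plugin.
# It plays in the Athena ambient sound class with overridden concurrency and virtualization disabled.
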SoundCue /TidalCraneDropPod/Assets/Audio/SC_TidalCraneDropPod_EnterRift
root=Mixer_0
nodes [
 {id=Mixer_0, children [Modulator_0,WavePlayer_1,Random_1,Random_2], properties {inputVolume [0.75,1,1,0.5]}}
 {id=Modulator_0, children [Random_0], properties {pitchMax=1.3, pitchMin=0.9, volumeMax=1, volumeMin=1}}
 {id=Random_0, children [WavePlayer_5,WavePlayer_6], properties {bShouldExcludeFromBranchCulling=true, weights [1,1]}}
 {id=Random_1, children [WavePlayer_2,WavePlayer_4], properties {weights [1,1]}}
 {id=Random_2, children [WavePlayer_0,WavePlayer_3,WavePlayer_7,WavePlayer_8], properties {weights [1,1,1,1]}}
 {id=WavePlayer_0, properties {soundWaveAssetPtr=/TidalCraneDropPod/Assets/Audio/wav/TidalCraneDropPod_EagleWhistle_01}}
 {id=WavePlayer_1, properties {soundWaveAssetPtr=/DropPodParentGameplay/Assets/Audio/wav/TidalCraneDropPod_EnterRiftA}}
 {id=WavePlayer_2, properties {soundWaveAssetPtr=/DropPodParentGameplay/Assets/Audio/wav/TidalCraneDropPod_EnterRiftB_01}}
 {id=WavePlayer_3, properties {soundWaveAssetPtr=/TidalCraneDropPod/Assets/Audio/wav/TidalCraneDropPod_EagleWhistle_02}}
 {id=WavePlayer_4, properties {soundWaveAssetPtr=/DropPodParentGameplay/Assets/Audio/wav/TidalCraneDropPod_EnterRiftB_02}}
 {id=WavePlayer_5, properties {soundWaveAssetPtr=/DropPodParentGameplay/Assets/Audio/wav/TidalCraneDropPod_EnterRiftC_01}}
 {id=WavePlayer_6, properties {soundWaveAssetPtr=/DropPodParentGameplay/Assets/Audio/wav/TidalCraneDropPod_EnterRiftC_02}}
 {id=WavePlayer_7, properties {soundWaveAssetPtr=/TidalCraneDropPod/Assets/Audio/wav/TidalCraneDropPod_EagleWhistle_03}}
 {id=WavePlayer_8, properties {soundWaveAssetPtr=/TidalCraneDropPod/Assets/Audio/wav/TidalCraneDropPod_EagleWhistle_04}}
]
properties {
 attenuationSettings=/DropPodParentGameplay/Assets/Audio/ATT_TidalCraneDropPod_Alert_HighFreq
 bOverrideConcurrency=true
 maxDistance=230000
 soundClassObject=&Fort_Ambient_Athena
 virtualizationMode=Disabled
 volumeMultiplier=1
 concurrencyOverrides {maxCount=4}
}
