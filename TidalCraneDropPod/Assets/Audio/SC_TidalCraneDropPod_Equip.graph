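# Sound cue for the equip sound of the Tidal Crane Drop Pod item.
# A Random node at the root picks one of six equip wave variations with equal weight.
# It plays at reduced volume in the gadgets and items sound class, using the weapon reload attenuation settings.
# A concurrency override allows one instance at a time and blocks new plays while it is active.
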SoundCue /TidalCraneDropPod/Assets/Audio/SC_TidalCraneDropPod_Equip
root=Random_0
nodes [
 {id=Random_0, children [WavePlayer_0,WavePlayer_1,WavePlayer_2,WavePlayer_6,WavePlayer_7,WavePlayer_8], properties {weights [1,1,1,1,1,1]}}
 {id=WavePlayer_0, properties {soundWaveAssetPtr=/TidalCraneDropPod/Assets/Audio/wav/TidalCraneDropPod_Equip_01}}
 {id=WavePlayer_1, properties {soundWaveAssetPtr=/TidalCraneDropPod/Assets/Audio/wav/TidalCraneDropPod_Equip_02}}
 {id=WavePlayer_2, properties {soundWaveAssetPtr=/TidalCraneDropPod/Assets/Audio/wav/TidalCraneDropPod_Equip_03}}
 {id=WavePlayer_6, properties {soundWaveAssetPtr=/TidalCraneDropPod/Assets/Audio/wav/TidalCraneDropPod_Equip_04}}
 {id=WavePlayer_7, properties {soundWaveAssetPtr=/TidalCraneDropPod/Assets/Audio/wav/TidalCraneDropPod_Equip_05}}
 {id=WavePlayer_8, properties {soundWaveAssetPtr=/TidalCraneDropPod/Assets/Audio/wav/TidalCraneDropPod_Equip_06}}
]
properties {
 attenuationSettings=&Fort_Weapon_Reload_Att
 maxDistance=2250
 soundClassObject=&Fort_Gadgets_Items
 volumeMultiplier=0.55
 concurrencyOverrides {maxCount=1, resolutionRule=PreventNew}
}
