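# Sound cue for the Tidal Crane drop pod's flare activation.
# A mixer layers one of three equally weighted, randomly chosen activate-and-land variants over a fixed activation sound, which plays at half input volume.
# It uses flare gun projectile alert attenuation and a third-person ranged weapon fire sound class, with reduced overall volume and a concurrency limit override.
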SoundCue /TidalCraneDropPod/Assets/Audio/SC_TidalCraneDropPod_FlareActivate
root=Mixer_0
nodes [
 {id=Mixer_0, children [Random_1,WavePlayer_1], properties {inputVolume [1,0.5]}}
 {id=Random_1, children [WavePlayer_3,WavePlayer_4,WavePlayer_5], properties {weights [1,1,1]}}
 {id=WavePlayer_1, properties {soundWaveAssetPtr=/TidalCraneDropPod/Assets/Audio/wav/TidalCraneDropPod_FlareActivate}}
 {id=WavePlayer_3, properties {soundWaveAssetPtr=/TidalCraneDropPod/Assets/Audio/wav/TidalCraneDropPod_FlareActivateLand_03}}
 {id=WavePlayer_4, properties {soundWaveAssetPtr=/TidalCraneDropPod/Assets/Audio/wav/TidalCraneDropPod_FlareActivateLand_02}}
 {id=WavePlayer_5, properties {soundWaveAssetPtr=/TidalCraneDropPod/Assets/Audio/wav/TidalCraneDropPod_FlareActivateLand_01}}
]
properties {
 attenuationSettings=/Game/Sounds/Attenuation/Fort_Weapon_Projectile_FlareGun_Alert_Att
 maxDistance=10000
 soundClassObject=&Fort_WeaponRanged_Fire_3P_PSM
 volumeMultiplier=0.35
 concurrencyOverrides {maxCount=4}
}
