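# Sound cue for the TidalCraneDropPod flare activating in water, layering an activation sound with a water impact.
# A Mixer sums two Random branches: one of three land-activation waves at a reduced level, and one of five water-impact waves at a boosted level.
# It borrows the flare gun alert attenuation, uses a third-person ranged weapon fire sound class, and caps concurrent playback.
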SoundCue /TidalCraneDropPod/Assets/Audio/SC_TidalCraneDropPod_FlareActivateInWater
root=Mixer_1
nodes [
 {id=Mixer_1, children [Random_3,Random_2], properties {inputVolume [0.8,1.4]}}
 {id=Random_2, children [WavePlayer_0,WavePlayer_1,WavePlayer_2,WavePlayer_7,WavePlayer_8], properties {weights [1,1,1,1,1]}}
 {id=Random_3, children [WavePlayer_9, WavePlayer_10, WavePlayer_11], properties {weights [1,1,1]}}
 {id=WavePlayer_0, properties {soundWaveAssetPtr=/TidalCraneDropPod/Assets/Audio/wav/TidalCraneDropPod_Flare_ImpactWater_05}}
 {id=WavePlayer_1, properties {soundWaveAssetPtr=/TidalCraneDropPod/Assets/Audio/wav/TidalCraneDropPod_Flare_ImpactWater_04}}
 {id=WavePlayer_10, properties {soundWaveAssetPtr=/TidalCraneDropPod/Assets/Audio/wav/TidalCraneDropPod_FlareActivateLand_02}}
 {id=WavePlayer_11, properties {soundWaveAssetPtr=/TidalCraneDropPod/Assets/Audio/wav/TidalCraneDropPod_FlareActivateLand_01}}
 {id=WavePlayer_2, properties {soundWaveAssetPtr=/TidalCraneDropPod/Assets/Audio/wav/TidalCraneDropPod_Flare_ImpactWater_03}}
 {id=WavePlayer_7, properties {soundWaveAssetPtr=/TidalCraneDropPod/Assets/Audio/wav/TidalCraneDropPod_Flare_ImpactWater_02}}
 {id=WavePlayer_8, properties {soundWaveAssetPtr=/TidalCraneDropPod/Assets/Audio/wav/TidalCraneDropPod_Flare_ImpactWater_01}}
 {id=WavePlayer_9, properties {soundWaveAssetPtr=/TidalCraneDropPod/Assets/Audio/wav/TidalCraneDropPod_FlareActivateLand_03}}
]
properties {
 attenuationSettings=/Game/Sounds/Attenuation/Fort_Weapon_Projectile_FlareGun_Alert_Att
 maxDistance=10000
 soundClassObject=&Fort_WeaponRanged_Fire_3P_PSM
 volumeMultiplier=0.5
 concurrencyOverrides {maxCount=4}
}
