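# Sound cue for the Tidal Crane Drop Pod flare while it is in the air, using flare gun alert attenuation with capped concurrency.
# A mixer layers randomized looping drop-pod flare and flare-gun loops, each through a volume-ramping enveloper, with a one-shot in-air sound and a randomly picked whistle.
# The in-air one-shot is boosted in the mix, while both loop layers are reduced.
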SoundCue /TidalCraneDropPod/Assets/Audio/SC_TidalCraneDropPod_FlareInAir
root=Mixer_0
nodes [
 {
  id=Enveloper_0
  children [Random_0]
  properties {
   pitchCurve {editorCurveData {keys [{time=0}]}}
   volumeCurve {editorCurveData {keys [<at 0 0.2049>, {arriveTangent=0.7951, interpMode=RCIM_Cubic, leaveTangent=0.7951, time=0.4436, value=0.779}, <at 1 1>]}}
  }
 }
 {
  id=Enveloper_1
  children [Random_2]
  properties {pitchCurve {editorCurveData {keys [{time=0}]}}, volumeCurve {editorCurveData {keys [<at 0 0.2049>,<at 1 1>]}}}
 }
 {id=Mixer_0, children [Enveloper_0,Enveloper_1,WavePlayer_9,Random_4], properties {inputVolume [0.5,0.5,2,1]}}
 {id=Random_0, children [WavePlayer_0,WavePlayer_1,WavePlayer_2], properties {weights [1,1,1]}}
 {id=Random_2, children [WavePlayer_12, WavePlayer_13], properties {weights [1,1]}}
 {id=Random_4, children [WavePlayer_16, WavePlayer_17, WavePlayer_18, WavePlayer_19], properties {weights [1,1,1,1]}}
 {id=WavePlayer_0, properties {bLooping=true, soundWaveAssetPtr=/TidalCraneDropPod/Assets/Audio/wav/TidalCraneDropPod_Flare_InAir_Loop_03}}
 {id=WavePlayer_1, properties {bLooping=true, soundWaveAssetPtr=/TidalCraneDropPod/Assets/Audio/wav/TidalCraneDropPod_Flare_InAir_Loop_02}}
 {id=WavePlayer_12, properties {bLooping=true, soundWaveAssetPtr=/TidalCraneDropPod/Assets/Audio/wav/FlareGun_InAirLoop_01}}
 {id=WavePlayer_13, properties {bLooping=true, soundWaveAssetPtr=/TidalCraneDropPod/Assets/Audio/wav/FlareGun_InAirLoop_02}}
 {id=WavePlayer_16, properties {soundWaveAssetPtr=/TidalCraneDropPod/Assets/Audio/wav/TidalCraneDropPod_FlareWhistle_01}}
 {id=WavePlayer_17, properties {soundWaveAssetPtr=/TidalCraneDropPod/Assets/Audio/wav/TidalCraneDropPod_FlareWhistle_02}}
 {id=WavePlayer_18, properties {soundWaveAssetPtr=/TidalCraneDropPod/Assets/Audio/wav/TidalCraneDropPod_FlareWhistle_03}}
 {id=WavePlayer_19, properties {soundWaveAssetPtr=/TidalCraneDropPod/Assets/Audio/wav/TidalCraneDropPod_FlareWhistle_04}}
 {id=WavePlayer_2, properties {bLooping=true, soundWaveAssetPtr=/TidalCraneDropPod/Assets/Audio/wav/TidalCraneDropPod_Flare_InAir_Loop_01}}
 {id=WavePlayer_9, properties {soundWaveAssetPtr=/TidalCraneDropPod/Assets/Audio/wav/TidalCraneDropPod_FlareInAir}}
]
properties {
 attenuationSettings=/Game/Sounds/Attenuation/Fort_Weapon_Projectile_FlareGun_Alert_Att
 maxDistance=10000
 soundClassObject=&Fort_WeaponRanged_Fire_3P_PSM
 volumeMultiplier=0.4
 concurrencyOverrides {maxCount=4}
}
