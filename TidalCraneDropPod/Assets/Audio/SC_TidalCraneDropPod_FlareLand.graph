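# Sound cue for the Tidal Crane drop pod's flare landing sound.
# A single wave player plays the pod's Land wave through a modulator that randomizes pitch within a narrow, lowered range.
# The cue plays at reduced volume, routed to the gadget items sound class with weapon-drop attenuation.
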SoundCue /TidalCraneDropPod/Assets/Audio/SC_TidalCraneDropPod_FlareLand
root=Modulator_0
nodes [
 {id=Modulator_0, children [WavePlayer_1], properties {pitchMax=0.9, pitchMin=0.8}}
 {id=WavePlayer_1, properties {soundWaveAssetPtr=/TidalCraneDropPod/Assets/Audio/wav/TidalCraneDropPod_Land}}
]
properties {
 attenuationSettings=&Fort_Weapon_Drop_Att
 maxDistance=3000
 soundClassObject=&Fort_Gadgets_Items
 volumeMultiplier=0.25
}
