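# Sound cue for the Tidal Crane Drop Pod's in-air state, using gadget-activate attenuation and the weapons sound class.
# A Mixer layers four branches at separate input volumes: a pitched-up one-shot beep, a looping beep, a random flare-gun in-air loop and a random flare activate/land one-shot.
# An Enveloper fades the looping beep in over a short volume curve, so it builds beneath the initial one-shots.
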SoundCue /TidalCraneDropPod/Assets/Audio/SC_TidalCraneDropPod_InAir
root=Mixer_1
nodes [
 {
  id=Enveloper_0
  children [Modulator_2]
  properties {
   pitchCurve {editorCurveData {keys [{time=0}]}}
   volumeCurve {editorCurveData {keys [<at 0 0 RCIM_Cubic>, {arriveTangent=0.1401, interpMode=RCIM_Cubic, leaveTangent=0.1401, tangentMode=RCTM_User, time=0.6, value=1}]}}
  }
 }
 {id=Mixer_1, children [Modulator_0,Enveloper_0,Random_2,Random_0], properties {inputVolume [0.35,1,0.5,0.5]}}
 {id=Modulator_0, children [WavePlayer_4], properties {pitchMax=1.3, pitchMin=1.3}}
 {id=Modulator_2, children [WavePlayer_3], properties {pitchMax=1, pitchMin=1, volumeMax=1, volumeMin=1}}
 {id=Random_0, children [WavePlayer_6,WavePlayer_7,WavePlayer_8], properties {weights [1,1,1]}}
 {id=Random_2, children [WavePlayer_1,WavePlayer_5], properties {weights [1,1]}}
 {id=WavePlayer_1, properties {bLooping=true, soundWaveAssetPtr=/TidalCraneDropPod/Assets/Audio/wav/FlareGun_InAirLoop_01}}
 {id=WavePlayer_3, properties {bLooping=true, soundWaveAssetPtr=/TidalCraneDropPod/Assets/Audio/wav/TidalCraneDropPod_Beep_01}}
 {id=WavePlayer_4, properties {soundWaveAssetPtr=/TidalCraneDropPod/Assets/Audio/wav/TidalCraneDropPod_Beep_02}}
 {id=WavePlayer_5, properties {bLooping=true, soundWaveAssetPtr=/TidalCraneDropPod/Assets/Audio/wav/FlareGun_InAirLoop_02}}
 {id=WavePlayer_6, properties {soundWaveAssetPtr=/TidalCraneDropPod/Assets/Audio/wav/TidalCraneDropPod_FlareActivateLand_01}}
 {id=WavePlayer_7, properties {soundWaveAssetPtr=/TidalCraneDropPod/Assets/Audio/wav/TidalCraneDropPod_FlareActivateLand_02}}
 {id=WavePlayer_8, properties {soundWaveAssetPtr=/TidalCraneDropPod/Assets/Audio/wav/TidalCraneDropPod_FlareActivateLand_03}}
]
properties {
 attenuationSettings=&Fort_GadgetActivate_Att
 maxDistance=4000
 soundClassObject=/Game/Sounds/FortSoundClassesAndModes/Fort_Weapons
 volumeMultiplier=1
}
