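# Sound cue for the pickup sound of the Tidal Crane Drop Pod item.
# A single wave player feeds the pickup wave straight to the root, with no randomization or modulation.
# It uses Fortnite weapon-pickup attenuation and the gadget items sound class, with reduced volume and a limited audible distance.
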SoundCue /TidalCraneDropPod/Assets/Audio/SC_TidalCraneDropPod_Pickup
root=WavePlayer_1
nodes [{id=WavePlayer_1, properties {soundWaveAssetPtr=/TidalCraneDropPod/Assets/Audio/wav/TidalCraneDropPod_Pickup}}]
properties {
 attenuationSettings=&Fort_Weapon_Pickup_Att
 maxDistance=3500
 soundClassObject=&Fort_Gadgets_Items
 volumeMultiplier=0.65
}
