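# Sound cue for throwing the Tidal Crane Drop Pod.
# A mixer layers three equally weighted random pools: pod-specific throw sounds, louder throw whooshes, and two EMP grenade throw clips.
# It borrows weapon-reload attenuation and a first-person ranged weapon fire sound class, with reduced overall volume and a limited audible range.
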SoundCue /TidalCraneDropPod/Assets/Audio/SC_TidalCraneDropPod_Throw
root=Mixer_0
nodes [
 {id=Mixer_0, children [Random_0,Random_2,Random_4], properties {inputVolume [0.75,1.5,1]}}
 {id=Random_0, children [WavePlayer_3,WavePlayer_4,WavePlayer_5], properties {weights [1,1,1]}}
 {id=Random_2, children [WavePlayer_6,WavePlayer_7,WavePlayer_8], properties {weights [1,1,1]}}
 {id=Random_4, children [WavePlayer_0,WavePlayer_2], properties {weights [1,1]}}
 {id=WavePlayer_0, properties {soundWaveAssetPtr=/TidalCraneDropPod/Assets/Audio/wav/EMP_Grenade_Throw_01}}
 {id=WavePlayer_2, properties {soundWaveAssetPtr=/TidalCraneDropPod/Assets/Audio/wav/EMP_Grenade_Throw_03}}
 {id=WavePlayer_3, properties {soundWaveAssetPtr=/TidalCraneDropPod/Assets/Audio/wav/TidalCraneDropPod_Throw_03}}
 {id=WavePlayer_4, properties {soundWaveAssetPtr=/TidalCraneDropPod/Assets/Audio/wav/TidalCraneDropPod_Throw_02}}
 {id=WavePlayer_5, properties {soundWaveAssetPtr=/TidalCraneDropPod/Assets/Audio/wav/TidalCraneDropPod_Throw_01}}
 {id=WavePlayer_6, properties {soundWaveAssetPtr=/TidalCraneDropPod/Assets/Audio/wav/Throw_Whoosh_03}}
 {id=WavePlayer_7, properties {soundWaveAssetPtr=/TidalCraneDropPod/Assets/Audio/wav/Throw_Whoosh_02}}
 {id=WavePlayer_8, properties {soundWaveAssetPtr=/TidalCraneDropPod/Assets/Audio/wav/Throw_Whoosh_01}}
]
properties {
 attenuationSettings=&Fort_Weapon_Reload_Att
 maxDistance=2250
 soundClassObject=&Fort_WeaponRanged_Fire_1P_PSM
 volumeMultiplier=0.6
}
