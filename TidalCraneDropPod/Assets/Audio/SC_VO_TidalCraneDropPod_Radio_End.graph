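# Sound cue for the radio end sound of the Tidal Crane drop pod, routed through the enemy non-critical VO sound class.
# An IsLocal branch plays the wave for the local case, otherwise an IsEnemy branch silences enemies through an empty zero-volume mixer and plays it for non-enemies.
# Unset parameters also play the wave.
# All audible paths share one wave player and the flare gun alert attenuation, via two identical attenuation nodes.
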SoundCue /TidalCraneDropPod/Assets/Audio/SC_VO_TidalCraneDropPod_Radio_End
root=Branch_1
nodes [
 {
  id=Attenuation_0
  children [WavePlayer_0]
  properties {attenuationSettings=/Game/Sounds/Attenuation/Fort_Weapon_Projectile_FlareGun_Alert_Att}
 }
 {
  id=Attenuation_2
  children [WavePlayer_0]
  properties {attenuationSettings=&Fort_Weapon_Projectile_FlareGun_Alert_Att}
 }
 {id=Branch_1, children [Attenuation_0, Branch_2, Attenuation_0], properties {boolParameterName=IsLocal}}
 {id=Branch_2, children [Mixer_8, Attenuation_2, Attenuation_0], properties {boolParameterName=IsEnemy}}
 {id=Mixer_8, children [null], properties {inputVolume [0]}}
 {id=WavePlayer_0, properties {soundWaveAssetPtr=/TidalCraneDropPod/Assets/Audio/wav/DropPod_RadioEnd}}
]
properties {maxDistance=10000, soundClassObject=/Game/Sounds/FortSoundClassesAndModes/Fort_VO_Enemy_NonCritical, volumeMultiplier=1}
