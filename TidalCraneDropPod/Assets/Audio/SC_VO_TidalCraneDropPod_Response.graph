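# Sound Cue that plays the walkie-talkie voice response for the TidalCraneDropPod, routed through an enemy non-critical VO sound class.
# After a Delay node, branches on IsLocal and IsEnemy give local playback a louder mix, non-enemies a normal mix, and enemies silence.
# Both audible paths randomly pick from the same four response waves with equal weight and share a flare-gun alert attenuation.
# Unset parameters fall back to the louder local path.
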SoundCue /TidalCraneDropPod/Assets/Audio/SC_VO_TidalCraneDropPod_Response
root=Delay_0
nodes [
 {id=Attenuation_0, children [Mixer_0], properties {attenuationSettings=/Game/Sounds/Attenuation/Fort_Weapon_Projectile_FlareGun_Alert_Att}}
 {id=Attenuation_2, children [Mixer_6], properties {attenuationSettings=&Fort_Weapon_Projectile_FlareGun_Alert_Att}}
 {id=Branch_1, children [Attenuation_0, Branch_2, Attenuation_0], properties {boolParameterName=IsLocal}}
 {id=Branch_2, children [Mixer_8, Attenuation_2, Attenuation_0], properties {boolParameterName=IsEnemy}}
 {id=Delay_0, children [Branch_1]}
 {id=Mixer_0, children [Random_0], properties {inputVolume [1.2]}}
 {id=Mixer_6, children [Random_3], properties {inputVolume [1]}}
 {id=Mixer_8, children [null], properties {inputVolume [0]}}
 {id=Random_0, children [WavePlayer_0,WavePlayer_1,WavePlayer_2,WavePlayer_3], properties {weights [1,1,1,1]}}
 {id=Random_3, children [WavePlayer_10, WavePlayer_11, WavePlayer_12, WavePlayer_13], properties {weights [1,1,1,1]}}
 {id=WavePlayer_0, properties {soundWaveAssetPtr=/TidalCraneDropPod/Assets/Audio/wav/TidalCraneDropPod_WalkieTalkieResponse_01}}
 {id=WavePlayer_1, properties {soundWaveAssetPtr=/TidalCraneDropPod/Assets/Audio/wav/TidalCraneDropPod_WalkieTalkieResponse_02}}
 {id=WavePlayer_10, properties {soundWaveAssetPtr=&TidalCraneDropPod_WalkieTalkieResponse_01}}
 {id=WavePlayer_11, properties {soundWaveAssetPtr=&TidalCraneDropPod_WalkieTalkieResponse_02}}
 {id=WavePlayer_12, properties {soundWaveAssetPtr=/TidalCraneDropPod/Assets/Audio/wav/TidalCraneDropPod_WalkieTalkieResponse_03}}
 {id=WavePlayer_13, properties {soundWaveAssetPtr=/TidalCraneDropPod/Assets/Audio/wav/TidalCraneDropPod_WalkieTalkieResponse_04}}
 {id=WavePlayer_2, properties {soundWaveAssetPtr=&TidalCraneDropPod_WalkieTalkieResponse_03}}
 {id=WavePlayer_3, properties {soundWaveAssetPtr=&TidalCraneDropPod_WalkieTalkieResponse_04}}
]
properties {maxDistance=10000, soundClassObject=/Game/Sounds/FortSoundClassesAndModes/Fort_VO_Enemy_NonCritical, volumeMultiplier=1}
