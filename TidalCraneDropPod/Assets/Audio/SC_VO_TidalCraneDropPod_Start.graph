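# Sound cue for the TidalCraneDropPod start voice line, playing one of three random walkie-talkie start sounds on an enemy non-critical VO class.
# An IsLocal branch plays the sound when true or unset, and otherwise defers to an IsEnemy branch that silences it for enemies through an empty zero-volume mixer.
# The audible paths are duplicate random chains sharing the same waves and flare gun alert attenuation.
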SoundCue /TidalCraneDropPod/Assets/Audio/SC_VO_TidalCraneDropPod_Start
root=Branch_1
nodes [
 {id=Attenuation_0, children [Mixer_0], properties {attenuationSettings=/Game/Sounds/Attenuation/Fort_Weapon_Projectile_FlareGun_Alert_Att}}
 {id=Attenuation_2, children [Mixer_6], properties {attenuationSettings=&Fort_Weapon_Projectile_FlareGun_Alert_Att}}
 {id=Branch_1, children [Attenuation_0, Branch_2, Attenuation_0], properties {boolParameterName=IsLocal}}
 {id=Branch_2, children [Mixer_8, Attenuation_2, Attenuation_0], properties {boolParameterName=IsEnemy}}
 {id=Mixer_0, children [Random_1], properties {inputVolume [1]}}
 {id=Mixer_6, children [Random_2], properties {inputVolume [1]}}
 {id=Mixer_8, children [null], properties {inputVolume [0]}}
 {id=Random_1, children [WavePlayer_7,WavePlayer_8,WavePlayer_9], properties {weights [1,1,1]}}
 {id=Random_2, children [WavePlayer_3,WavePlayer_4,WavePlayer_5], properties {weights [1,1,1]}}
 {id=WavePlayer_3, properties {soundWaveAssetPtr=/TidalCraneDropPod/Assets/Audio/wav/TidalCraneDropPod_WalkieTalkieStart_01}}
 {id=WavePlayer_4, properties {soundWaveAssetPtr=/TidalCraneDropPod/Assets/Audio/wav/TidalCraneDropPod_WalkieTalkieStart_02}}
 {id=WavePlayer_5, properties {soundWaveAssetPtr=/TidalCraneDropPod/Assets/Audio/wav/TidalCraneDropPod_WalkieTalkieStart_03}}
 {id=WavePlayer_7, properties {soundWaveAssetPtr=&TidalCraneDropPod_WalkieTalkieStart_01}}
 {id=WavePlayer_8, properties {soundWaveAssetPtr=&TidalCraneDropPod_WalkieTalkieStart_02}}
 {id=WavePlayer_9, properties {soundWaveAssetPtr=&TidalCraneDropPod_WalkieTalkieStart_03}}
]
properties {maxDistance=10000, soundClassObject=/Game/Sounds/FortSoundClassesAndModes/Fort_VO_Enemy_NonCritical, volumeMultiplier=1}
