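# Post-process danger alert material for the Tidal Crane Drop Pod that tints part of the screen with pulsing red.
# A sine of time scaled by PulseRate brightens the red, which is blended over the scene color.
# An inverted screen-position smoothstep, with axis set by static WipeDirection mask, confines the tint by default to the top half, peaking at half strength.
# BlendWeight from the MPC_PPM parameter collection scales the whole effect.
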Material /TidalCraneDropPod/Assets/FX/Materials/PPM_DangerAlert_TidalCraneDropPod
graph {
 properties {bAutomaticallySetUsageInEditor=false, materialDomain=MD_PostProcess, usedShadingModels=Default Lit}
 nodes [
  n1 = Time()
  n2 = ScalarParameter() {defaultValue=0.75, parameterName=PulseRate}
  n3 = ConstantBiasScale(None=Sine(None=Multiply(A=@n1, B=@n2)))
  n4 = Constant() {r=1}
  n5 = Constant() {r=0.5}
  n6 = Subtract(A=@n4, B=@n5)
  n7 = Add(A=@n4, B=@n5)
  n8 = OneMinus(None=StaticComponentMaskParameter(None=ScreenPosition().ViewportUV) {defaultG=true, parameterName=WipeDirection})
  LinearInterpolate(
   A=ComponentMask(None=SceneTexture() {sceneTextureId=PPI_PostProcessInput0}.Color) {b=true, g=true, r=true}
   B=Add(A=LinearInterpolate(Alpha=@n3) {constB=0.2}, B=Constant3Vector() {constant=<rgba 1 0 0 0>})
   Alpha=LinearInterpolate(
    B=SmoothStep(Min=@n6, Max=@n7, Value=@n8) {constMax=0.05, constMin=0.05}
    Alpha=CollectionParameter() {collection=/TidalCraneDropPod/DataTables/MPC_PPM, parameterName=BlendWeight}
   )
  )
 ]
}
